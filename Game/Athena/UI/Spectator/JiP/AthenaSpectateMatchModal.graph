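# Spectate-a-match modal for joining a match through a friend's session or a game session ID, using parent-class calls.
# Two buttons toggle mutually exclusive sections: a friend list and an ID field whose Join button enables only for valid IDs, though Enter submits unchecked.
# Friend names become dynamic entry buttons mapped to names for joining; an empty, visible list starts a cooldown timer.
# Custom navigation functions route focus between sections.
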WidgetBlueprint /Game/Athena/UI/Spectator/JiP/AthenaSpectateMatchModal
graphs [
 EventGraph {nodes [
   "On Clicked (JoinGameSessionButton)"!() -> (OutputDelegate, then, Button:Common Button Base Object Reference) {is=title K2Node_ComponentBoundEvent}
   n1 = Get GameSessionHB() -> (GameSessionHB:Horizontal Box Object Reference)
   > n2 = Branch(Condition=IsVisible(self=@n1))
   > SetVisibility(self=@n1, InVisibility=Collapsed)
   > CenterWidget!(execute, ToCenter:Widget Object Reference=Get JoinGameSessionButton() -> (JoinGameSessionButton:WBP UIKit Button Regular Object Reference)) -> (then)
   SetVisibility(execute=@n2.else, self=@n1, InVisibility=SelfHitTestInvisible)
   > SetVisibility(self=Get FriendSwitcher() -> (FriendSwitcher:Common Widget Switcher Legacy Object Reference), InVisibility=Collapsed)
   > Clear and Invalidate Timer by Handle(Handle=Get FriendListCooldownTimer() -> (FriendListCooldownTimer:Timer Handle Structure))
   n3 = Get GameSessionSearchText() -> (GameSessionSearchText:WBP UIKit Input Field Object Reference)
   > CenterWidget(ToCenter=Get EditableText!(self:WBP UIKit Block Editable Text Object Reference=Get Block_EditableText!(self:WBP UIKit Input Field Base Object Reference=@n3) -> (Block_EditableText:WBP UIKit Block Editable Text Object Reference)) -> (EditableText:UIKit Editable Text Object Reference))
   n4 = "On Clicked (CloseButton)"!() -> (OutputDelegate, then, Button:Common Button Base Object Reference) {is=title K2Node_ComponentBoundEvent}
   DeactivateWidget(execute=[@n4.then, @n7.then])
   "On Clicked (JoinFriendSessionButton)"!() -> (OutputDelegate, then, Button:Common Button Base Object Reference) {is=title K2Node_ComponentBoundEvent}
   n5 = Get FriendSwitcher()
   > n6 = Branch(Condition=IsVisible(self=@n5))
   > SetVisibility(self=@n5, InVisibility=Collapsed)
   SetVisibility(execute=@n6.else, self=@n5, InVisibility=SelfHitTestInvisible)
   > SetVisibility(self=Get GameSessionHB(), InVisibility=Collapsed)
   > RefreshFriendList()
   n7 = "On Button Clicked (Button_CloseTouch)"!() -> (OutputDelegate, then, Button:Common Button Legacy Object Reference) {is=title K2Node_ComponentBoundEvent}
   "On Clicked (JoinButton)"!() -> (OutputDelegate, then, Button:Common Button Base Object Reference) {is=title K2Node_ComponentBoundEvent}
   n8 = Get EditableText(self=Get Block_EditableText(self=Get GameSessionSearchText()))
   > JoinGameSession!(
    execute
    self:Athena Spectate Match Modal Base Object Reference
    GameSessionIdStr:String="To String (Text)"(InText="GetText (Editable Text)"!(self:Editable Text Object Reference=@n8) -> (ReturnValue:Text))
   ) -> (then)
   n9 = "On Text Commited (GameSessionSearchText)"!() -> (OutputDelegate, then, Text:Text, Commit method:ETextCommit Enum=Default) {is=title K2Node_ComponentBoundEvent}
   > Branch(Condition="Equal (Enum)"(A:ETextCommit Enum=@n9.Commit method, B:ETextCommit Enum=OnEnter))
   > JoinGameSession(GameSessionIdStr="To String (Text)"(InText=@n9.Text))
   n10 = "On Text Changed (GameSessionSearchText)"!() -> (OutputDelegate, then, Text:Text) {is=title K2Node_ComponentBoundEvent}
   > SetIsEnabled(
    self=Get JoinButton() -> (JoinButton:WBP UIKit Button Regular Object Reference)
    bInIsEnabled=IsGameSessionString!(self:Athena Spectate Match Modal Base Object Reference, String:String="To String (Text)"(InText=@n10.Text)) -> (ReturnValue:bool)
   )
   n11 = CreateFriendList!() -> (OutputDelegate, then, FriendsNames:Array of Strings) {is=title K2Node_CustomEvent}
   > Reset!(
    execute
    self:Dynamic Entry Box Object Reference=Get FriendListEntryBox() -> (FriendListEntryBox:Dynamic Entry Box Object Reference)
    bDeleteWidgets:bool=false
   ) -> (then)
   n12 = Is Array Non Empty!(InputArray:Array of Strings=@n11.FriendsNames) -> (Non-Empty:bool=~) {is=title K2Node_MacroInstance}
   > SetActiveWidget(
    self:Widget Switcher Object Reference=Get FriendSwitcher()
    Widget=Select(
     Option 0:Widget Object Reference=Get NoFriendsPlayingText() -> (NoFriendsPlayingText:Common Text Object Reference)
     Option 1:Widget Object Reference=Get FriendListEntryBox()
     Index=@n12
    ) -> (ReturnValue:Widget Object Reference)
   )
   > n13 = Branch(Condition=@n12)
   > n14 = For Each Loop(Array=@n11.FriendsNames)
   > n15 = Create Entry!(execute, self:Dynamic Entry Box Object Reference=Get FriendListEntryBox()) -> (then, ReturnValue:User Widget Object Reference)
   > n16 = Cast To FortCTAButton!(execute, Object=@n15) -> (then, CastFailed, AsFort CTAButton:Fort CTAButton Object Reference) {is=title K2Node_DynamicCast}
   > Add!(
    execute
    TargetMap:Map of Common Button Base Object References to Strings=Get Friends Map() -> (Friends Map:Map of Common Button Base Object References to Strings)
    Key:"Common Button Base Object Reference (by ref)"=@n16
    Value:"String (by ref)"=@n14
   ) -> (then) {is=title K2Node_CallFunction}
   > Set Text(Text="To Text (String)"(InString=@n14), self:Fort CTAButton Object Reference=@n16)
   > Bind Event to On Clicked(self:Common Button Base Object Reference=@n16, Delegate=Create Event()) {is=title K2Node_AddDelegate}
   Branch(execute=@n13.else, Condition=IsVisible(self=Get FriendSwitcher()))
   > n17 = Set Timer by Event(Delegate=Create Event(), Time=Get FriendListCooldown() -> (FriendListCooldown:double), bLooping=false)
   > Set FriendListCooldownTimer(FriendListCooldownTimer:Timer Handle Structure=@n17) -> (Output_Get:Timer Handle Structure)
   RefreshFriendList() {is=title K2Node_CustomEvent}
   > BuildFriendNameIdMap(self:Athena Spectate Match Modal Base Object Reference)
   > CreateFriendList!(
    execute
    self
    FriendsNames:Array of Strings=GetFriendNames!(self:Athena Spectate Match Modal Base Object Reference) -> (FriendNames:Array of Strings)
   ) -> (then) {is=title K2Node_CallFunction}
   n18 = OnFriendNameClicked!() -> (OutputDelegate, then, ButtonClicked:Common Button Base Object Reference) {is=title K2Node_CustomEvent}
   > JoinFriendSession!(
    execute
    self:Athena Spectate Match Modal Base Object Reference
    FriendName:String=Find(
     TargetMap:Map of Common Button Base Object References to Strings=Get Friends Map()
     Key:"Common Button Base Object Reference (by ref)"=@n18.ButtonClicked
    )
   ) -> (then)
   Event Construct()
   > SetVisibility(self=[Get FriendSwitcher(), Get GameSessionHB()], InVisibility=Collapsed)
   > Reset(self=Get FriendListEntryBox(), bDeleteWidgets=false)
   n19 = Get GameSessionSearchText()
   > "SetText (Editable Text)"!(execute, self:Editable Text Object Reference=Get EditableText(self=Get Block_EditableText(self=@n19)), InText:Text) -> (then)
   > SetIsEnabled(self=Get JoinButton(), bInIsEnabled=false)
   Event Destruct()
   > Clear and Invalidate Timer by Handle(Handle=Get FriendListCooldownTimer())
   n20 = Event PreConstruct()
   > Branch(Condition=@n20.IsDesignTime)
   > CreateFriendList(FriendsNames=Make Array!("[0]":String=Friend 1, "[1]":String=Friend 2, "[2]":String=Friend 3, "[3]":String=Friend 4, "[4]":String=Friend 5) -> (Array:Array of Strings))
   Event On Activated()
   > RestoreCenterWidget!(
    execute
    self:Fort Activatable Panel Object Reference
    FallbackWidget:Widget Object Reference=Get JoinFriendSessionButton() -> (JoinFriendSessionButton:WBP UIKit Button Regular Object Reference)
   ) -> (then)
   Reroute Node(InputPin:Wildcard) -> (OutputPin:Wildcard)
  ]}
 JoinFriendSessionCustomNavigation {nodes [
   n1 = JoinFriendSessionCustomNavigation!() -> (then, Navigation:EUINavigation Enum) {is=title K2Node_FunctionEntry}
   > n2 = Switch on EUINavigation!(execute, Selection:EUINavigation Enum=@n1) -> (Left:Exec, Right:Exec, Up:Exec, Down:Exec, Next:Exec, Previous:Exec, Invalid:Exec) {is=title K2Node_SwitchEnum}
   n3 = Get FriendListEntryBox()
   n4 = Branch(execute=@n2.Down, Condition="integer > integer"(A=GetNumEntries!(self:Dynamic Entry Box Base Object Reference=@n3) -> (ReturnValue:int=0)))
   > Return Node!(execute, ReturnValue:Widget Object Reference=Get FriendListEntryBox())
   Return Node(execute=@n4.else, ReturnValue=Get JoinGameSessionButton())
  ]}
 JoinGameSessionButtonCustomNavigation {nodes [
   n1 = JoinGameSessionButtonCustomNavigation!() -> (then, Navigation:EUINavigation Enum) {is=title K2Node_FunctionEntry}
   > n2 = Switch on EUINavigation(Selection=@n1)
   Branch(execute=@n2.Down, Condition=IsVisible(self=Get GameSessionHB()))
   > Return Node(ReturnValue=Get GameSessionSearchText())
   n3 = Get FriendListEntryBox()
   n4 = Branch(execute=@n2.Up, Condition="integer > integer"(A=GetNumEntries(self=@n3) -> (ReturnValue=0)))
   > Return Node(ReturnValue=Get FriendListEntryBox())
   Return Node(execute=@n4.else, ReturnValue=Get JoinFriendSessionButton())
  ]}
]
variables [
 FriendListCooldown {type=number, default=2}
 FriendListCooldownTimer {type=TimerHandle}
 Friends Map {default {}}
]
widgets {
 info {
  inheritedWidgetCount=0
  namedSlotCount=2
  parentClass=/Script/FortniteUI.AthenaSpectateMatchModalBase
  rootWidgetClass=/FortUILibrary/Systems/Dialog/WBP_UIKit_Dialog_Internal.WBP_UIKit_Dialog_Internal_C
  widgetCount=16
 }
 widgets [
  WBP_UIKit_Dialog_Base {
   widgetClassPath=/FortUILibrary/Systems/Dialog/WBP_UIKit_Dialog_Internal
   bIsVariable=true
   widget=WBP_UIKit_Dialog_Base
   properties {
    useText=None
    useTitle=false
    dialogSIze {
     height Override {enable=false, value=800}
     max Aspect Ratio {enable=false, value=0}
     max Desired Height {enable=true, value=800}
     max Desired Width {enable=false, value=0}
     min Aspect Ratio {enable=false, value=0}
     min Desired Height {enable=false, value=0}
     min Desired Width {enable=false, value=0}
     width Override {enable=true, value=864}
    }
   }
  }
  Overlay_0 {widgetClassPath=&Overlay, namedSlotHost=WBP_UIKit_Dialog_Base, widget=Overlay_0}
  Content {
   widgetClassPath=&VerticalBox
   parent=Overlay_0
   widget=Content
   slot {class=OverlaySlot, properties {horizontalAlignment=HAlign_Fill, verticalAlignment=VAlign_Fill, padding=<ltrb 80 64 80 48>}}
  }
  Title {
   widgetClassPath=&CommonTextBlock
   parent=Content
   widget=Title
   properties {
    text=SPECTATE A MATCH
    font {
     bForceMonospaced=false
     bMaterialIsStencil=false
     fontMaterial=null
     fontObject=&HeadingNow
     letterSpacing=0
     monospacedWidth=1
     size=25.5
     skewAmount=0
     typefaceFontName=86 Bold Italic
     outlineSettings {
      bApplyOutlineToDropShadows=false
      bMiteredCorners=false
      bSeparateFillAlpha=false
      outlineMaterial=null
      outlineSize=0
      outlineColor=<rgba 0 0 0 1>
     }
    }
    shadowOffset=<0 0>
   }
   slot {class=VerticalBoxSlot, properties {padding=<ltrb 0 0 0 32>}}
  }
  JoinFriendSessionButton {
   widgetClassPath=&WBP_UIKit_Button_Regular
   bIsVariable=true
   parent=Content
   widget=JoinFriendSessionButton
   properties {
    navigation=&self:WidgetTree.JoinFriendSessionButton.WidgetNavigation_0
    text=Join Friend Session
   }
   slot {class=VerticalBoxSlot, properties {padding=<ltrb 0 0 0 16>}}
  }
  FriendSwitcher {
   widgetClassPath=&CommonWidgetSwitcherLegacy
   bIsVariable=true
   parent=Content
   widget=FriendSwitcher
   properties {navigation=&self:WidgetTree.FriendSwitcher.WidgetNavigation_0, transitionDuration=0}
   slot {class=VerticalBoxSlot, properties {padding=<ltrb 0 0 0 32>}}
  }
  NoFriendsPlayingText {
   widgetClassPath=&CommonTextBlock
   bIsVariable=true
   parent=FriendSwitcher
   widget=NoFriendsPlayingText
   properties {
    text=No friends currently playing
    font {
     bForceMonospaced=false
     bMaterialIsStencil=false
     fontMaterial=null
     fontObject=&HeadingNow
     letterSpacing=0
     monospacedWidth=1
     size=18
     skewAmount=0
     typefaceFontName=74 Regular
     outlineSettings {
      bApplyOutlineToDropShadows=false
      bMiteredCorners=false
      bSeparateFillAlpha=false
      outlineMaterial=null
      outlineSize=0
      outlineColor=<rgba 0 0 0 1>
     }
    }
    shadowOffset=<0 0>
   }
   slot {class=WidgetSwitcherSlot, properties {horizontalAlignment=HAlign_Center}}
  }
  FriendListEntryBox {
   widgetClassPath=&DynamicEntryBox
   bIsVariable=true
   parent=FriendSwitcher
   widget=FriendListEntryBox
   properties {
    entryBoxType=Vertical
    entryWidgetClass=/FortUILibrary/UIKit/Buttons/WBP_UIKit_Button_Quiet.WBP_UIKit_Button_Quiet_C
    maxElementSize=36
    navigation=&self:WidgetTree.FriendListEntryBox.WidgetNavigation_1
    numDesignerPreviewEntries=0
    entrySpacing=<0 16>
   }
   slot {class=WidgetSwitcherSlot}
  }
  JoinGameSessionButton {
   widgetClassPath=&WBP_UIKit_Button_Regular
   bIsVariable=true
   parent=Content
   widget=JoinGameSessionButton
   properties {
    navigation=&self:WidgetTree.JoinGameSessionButton.WidgetNavigation_0
    text=Join Game Session
   }
   slot {class=VerticalBoxSlot}
  }
  GameSessionHB {
   widgetClassPath=&HorizontalBox
   bIsVariable=true
   parent=Content
   widget=GameSessionHB
   slot {class=VerticalBoxSlot, properties {padding=<ltrb 0 16 0 0>}}
  }
  GameSessionSearchText {
   widgetClassPath=/FortUILibrary/UIKit/InputField/WBP_UIKit_InputField
   bIsVariable=true
   parent=GameSessionHB
   widget=GameSessionSearchText
   properties {
    hint Text=Enter a game session ID:
    holdData=&CommonUI_HoldData_Base_C
    navigation=&self:WidgetTree.GameSessionSearchText.WidgetNavigation_1
    useIcon=false
   }
   slot {class=HorizontalBoxSlot, properties {padding=<ltrb 0 0 16 0>, size {sizeRule=Fill, value=1}}}
  }
  JoinButton {
   widgetClassPath=&WBP_UIKit_Button_Regular
   bIsVariable=true
   parent=GameSessionHB
   widget=JoinButton
   properties {navigation=&self:WidgetTree.JoinButton.WidgetNavigation_0, text=Join}
   slot {class=HorizontalBoxSlot, properties {horizontalAlignment=HAlign_Right}}
  }
  VisibilityBorder_NonTouch {
   widgetClassPath=&CommonHardwareVisibilityBorder
   parent=Content
   widget=VisibilityBorder_NonTouch
   properties {
    horizontalAlignment=HAlign_Left
    verticalAlignment=VAlign_Bottom
    visibilityQuery {
     autoDescription=" NONE( Input.Touch )"
     tokenStreamVersion=0
     userDescription=""
     queryTokenStream [0,1,3,1,0]
     tagDictionary [{tagName=Input.Touch}]
    }
   }
   slot {class=VerticalBoxSlot, properties {horizontalAlignment=HAlign_Left, verticalAlignment=VAlign_Bottom, padding=<ltrb 0 64 0 0>}}
  }
  CloseButton {
   widgetClassPath=/FortUILibrary/UIKit/Buttons/WBP_UIKit_Button_Quiet
   bIsVariable=true
   parent=VisibilityBorder_NonTouch
   widget=CloseButton
   properties {bIsFocusable=false, text=Close, triggeringInputAction {dataTable=&InputActionDataTable, rowName=GenericBack}}
   slot {class=BorderSlot, properties {horizontalAlignment=HAlign_Left, verticalAlignment=VAlign_Bottom, padding=<ltrb 0 0 0 0>}}
  }
  VisibilityBorder_Touch {
   widgetClassPath=&CommonHardwareVisibilityBorder
   namedSlotHost=WBP_UIKit_Dialog_Base
   widget=VisibilityBorder_Touch
   properties {
    horizontalAlignment=HAlign_Right
    verticalAlignment=VAlign_Top
    visibilityQuery {
     autoDescription=" ANY( Input.Touch )"
     tokenStreamVersion=0
     userDescription=""
     queryTokenStream [0,1,1,1,0]
     tagDictionary [{tagName=Input.Touch}]
    }
   }
  }
  Button_CloseTouch {
   widgetClassPath=/FortUILibrary/Foundation/Buttons/CloseButton
   bIsVariable=true
   parent=VisibilityBorder_Touch
   widget=Button_CloseTouch
   properties {
    onButtonBaseClicked=(&self:WidgetTree.Button_CloseTouch.HandleOnButtonClicked)
    onButtonBaseDoubleClicked=(&self:WidgetTree.Button_CloseTouch.HandleOnButtonDoubleClicked)
    onButtonBaseHovered=(&self:WidgetTree.Button_CloseTouch.HandleOnButtonHovered)
    onButtonBaseUnhovered=(&self:WidgetTree.Button_CloseTouch.HandleOnButtonUnhovered)
    onSelectedChangedBase=(&self:WidgetTree.Button_CloseTouch.HandleOnSelectedChanged)
    renderTransform {angle=0, scale=<0.7 0.7>, shear=<0 0>, translation=<0 0>}
   }
   slot {class=BorderSlot, properties {horizontalAlignment=HAlign_Right, verticalAlignment=VAlign_Top, padding=<ltrb 0 0 0 0>}}
  }
 ]
}
